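; Bounce pad plant actor that accepts pawns outside vehicles and vehicles as bounce targets, and never dies from damage.
; For vehicles, it computes a launch velocity from scaled current velocity plus an upward boost.
; OnBounced calls BounceAnim on authority, whose remote branch runs a timeline driving scale and emissive parameters on a dynamic material.
; Server damage plays replicated melee or weapon impact cues, chosen by the pickaxe tag.
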(event EventBeginPlay
  (bind _staticmeshcomponent (Variables|Default|GetStaticMeshComponent))
  (bind _returnvalue (Rendering|Material|CreateDynamicMaterialInstance _staticmeshcomponent 0 0 "None" :!class /Script/Engine.PrimitiveComponent))
  (bind _output_get (Variables|Default|SetDynamicMat _returnvalue))
  (Rendering|Material|SetMaterial _staticmeshcomponent 0 _output_get))

(event Custom|BounceAnim
  (bind _dynamicmat (Variables|Default|GetDynamicMat))
  (Networking|SwitchHasAuthority
    (:Remote
      (bind (_direction _scale_xy _scale_z _emissivemult) (Timeline|Timeline PlayFromStart)
        (:Update
          (Rendering|Material|SetVectorParameterValue _dynamicmat "ScaleMult" (Math|Conversions|ToLinearColor(Vector) (Math|Vector|MakeVector _scale_xy _scale_xy _scale_z)) :!class /Script/Engine.MaterialInstanceDynamic)
          (Rendering|Material|SetScalarParameterValue _dynamicmat "EmissiveMult" _emissivemult :!class /Script/Engine.MaterialInstanceDynamic))))))

(event EventOnDamageServer
  (params (Damage :type Float%20%28single-precision%29) (DamageTags :type Gameplay%20Tag%20Container%20Structure%20%28by%20ref%29) (Momentum :type Vector :default "0, 0, 0") (HitInfo :type Hit%20Result%20Structure%20%28by%20ref%29) (InstigatedBy :type Controller%20Object%20Reference) (DamageCauser :type Actor%20Object%20Reference) (EffectContext :type Gameplay%20Effect%20Context%20Handle%20Structure))
  (bind _returnvalue (GameplayCue|MakeGameplayCueParametersfromHitResult HitInfo))
  ;[1 1 1] On Damage - play impact cues for pickaxe and weapons
  (if (GameplayTags|HasTag DamageTags (Variables|Default|GetPickaxeTag))
    (Ability|ExecuteGameplayCueWithParamsReplicated (Variables|Default|GetCue_Impact_Melee) _returnvalue)
    (else
      (Ability|ExecuteGameplayCueWithParamsReplicated (Variables|Default|GetCue_Impact_Weapon) _returnvalue))))

(fn BounceOther
  (params (Actor :type Actor%20Object%20Reference) (Direction :type Vector :default "0, 0, 0"))
  (bind _asfort_vehicle_interface (Utilities|Casting|CastToFortVehicleInterface Actor))
  (CallFunction|BounceVehicle _asfort_vehicle_interface))

(fn BounceVehicle
  (params (Vehicle :type Fort%20Vehicle%20Interface%20Interface))
  (Audio|BroadcastAbilityImpactatLocation 0 (Vehicle|GetActor Vehicle) (CallFunction|CalcVehicleLaunchVelocity Vehicle) 768.0))

(fn CalcVehicleLaunchVelocity
  (params (Vehicle :type Fort%20Vehicle%20Interface%20Interface))
  (return (+ (* (Vehicle|GetVehicleVelocity Vehicle) (ScalableFloat|GetValueatLevel (Variables|BounceProp|Vehicle|GetVehicleVelocityScalar))) (* "0.000000,0.000000,1.000000" (ScalableFloat|GetValueatLevel (Variables|BounceProp|Vehicle|GetVehicleUpBoost))))))

(fn CanBounceActor
  (params (Actor :type Actor%20Object%20Reference))
  (bind _asfort_pawn (Utilities|Casting|CastToFortPawn Actor)
    (:then
      (Utilities|IsinVehicle _asfort_pawn
        (:true
          (return false))
        (:false
          (return true))))
    (:CastFailed
      (Utilities|Casting|CastToFortVehicleInterface Actor
        (:then
          (return true))
        (:CastFailed
          (return false))))))

(fn NewFunction_0)

(fn OnBounced
  (params (BouncedActor :type Actor%20Object%20Reference))
  (|CallParentFunction BouncedActor)
  (Development|PrintString "Bounced")
  (Networking|SwitchHasAuthority
    (:Authority
      (CallFunction|BounceAnim))))

(fn ShouldDie
  (params (Damage :type Float%20%28single-precision%29) (EventInstigator :type Controller%20Object%20Reference) (DamageCauser :type Actor%20Object%20Reference) (bIsTest :type Boolean) (EffectContext :type Gameplay%20Effect%20Context%20Handle%20Structure%20%28by%20ref%29))
  (return false))

(fn ConstructionScript
  (|CallParentFunction))
