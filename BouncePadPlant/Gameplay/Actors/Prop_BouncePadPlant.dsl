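; Bounce pad plant prop that bounces Fort pawns and never dies, as ShouldDie always returns false.
; BeginPlay creates a dynamic material instance on the static mesh for the bounce effect.
; OnBounced calls the parent, prints a debug string and triggers BounceAnim on authority.
; BounceAnim plays its timeline only on the remote branch, animating the material's ScaleMult and EmissiveMult rather than transforming the mesh.
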(event EventBeginPlay
  (bind _staticmeshcomponent (Variables|Debug|GetStaticMeshComponent))
  (bind _returnvalue (Rendering|Material|CreateDynamicMaterialInstance _staticmeshcomponent 0 0 "None" :!class /Script/Engine.PrimitiveComponent))
  (bind _output_get (Variables|Default|SetDynamicMat _returnvalue))
  (Rendering|Material|SetMaterial _staticmeshcomponent 0 _output_get))

(event Custom|BounceAnim
  (bind _dynamicmat (Variables|Default|GetDynamicMat))
  (Networking|SwitchHasAuthority
    (:Remote
      (bind (_direction _scale_xy _scale_z _emissivemult) (Timeline|Timeline PlayFromStart)
        (:Update
          (Rendering|Material|SetVectorParameterValue _dynamicmat "ScaleMult" (Math|Conversions|ToLinearColor(Vector) (Math|Vector|MakeVector _scale_xy _scale_xy _scale_z)) :!class /Script/Engine.MaterialInstanceDynamic)
          (Rendering|Material|SetScalarParameterValue _dynamicmat "EmissiveMult" _emissivemult :!class /Script/Engine.MaterialInstanceDynamic))))))

(fn CanBounceActor
  (params (Actor :type Actor%20Object%20Reference))
  (Utilities|Casting|CastToFortPawn Actor
    (:then
      (return true))
    (:CastFailed
      (return false))))

(fn OnBounced
  (params (BouncedActor :type Actor%20Object%20Reference))
  (|CallParentFunction BouncedActor)
  (Development|PrintString "Bounced")
  (Networking|SwitchHasAuthority
    (:Authority
      (CallFunction|BounceAnim))))

(fn ShouldDie
  (params (Damage :type Float%20%28single-precision%29) (EventInstigator :type Controller%20Object%20Reference) (DamageCauser :type Actor%20Object%20Reference) (bIsTest :type Boolean) (EffectContext :type Gameplay%20Effect%20Context%20Handle%20Structure%20%28by%20ref%29))
  (return false))

(fn ConstructionScript
  (|CallParentFunction))
