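; Parent Blueprint for bouncy props that launch pawns overlapping their JumpTrigger, with CanBounceActor (always true) and an empty BounceOther as apparent override points.
; Launch velocity adds a scaled share of the actor's horizontal velocity to a world or actor-relative direction times a scalable, per-instance-adjusted magnitude.
; After each bounce it applies configured gameplay effects to the actor's ability system, and authority calls PlayBounceFX, which fires a local burst GameplayCue.
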(event OnComponentBeginOverlap(JumpTrigger)
  (params (OverlappedComponent :type Primitive%20Component%20Object%20Reference) (OtherActor :type Actor%20Object%20Reference) (OtherComp :type Primitive%20Component%20Object%20Reference) (OtherBodyIndex :type Integer) (bFromSweep :type Boolean) (SweepResult :type Hit%20Result%20Structure%20%28by%20ref%29))
  (CallFunction|HandleJumpTriggerOverlap OtherActor))

(event Custom|PlayBounceFX
  (Networking|CanExecuteCosmeticEvents
    (:True
      (CallFunction|ExecuteBounceCue))))

(fn ApplyBouncedGameplayEffects
  (params (BouncedActor :type Actor%20Object%20Reference))
  (bind _returnvalue (Ability|GetAbilitySystemComponent BouncedActor))
  (Utilities|IsValid _returnvalue
    (:Is%20Valid
      (for _array_element (Variables|BounceProp|GetOnBouncedGameplayEffects)
        (GameplayEffects|ApplyGameplayEffectToSelf :self _returnvalue :GameplayEffectClass _array_element :Level 0.0)))))

(fn BounceOther
  (params (Actor :type Actor%20Object%20Reference) (Direction :type Vector)))

(fn BouncePawn
  (params (Pawn :type Character%20Object%20Reference) (Direction :type Vector))
  (Character|LaunchCharacter Pawn Direction (Variables|BounceProp|PawnLaunch|GetPawnLaunchXYOverride) (Variables|BounceProp|PawnLaunch|GetPawnLaunchZOverride)))

(fn CanBounceActor
  (params (Actor :type Actor%20Object%20Reference))
  (return true))

(fn DoBounce
  (params (Actor :type Actor%20Object%20Reference))
  (Utilities|FlowControl|Sequence
    (:then_0
      (bind _returnvalue (CallFunction|GetBounceVelocity Actor))
      (bind _output_get (LocalVariables|SetBounceDirection _returnvalue))
      (bind _asfort_pawn (Utilities|Casting|CastToFortPawn Actor)
        (:then
          (CallFunction|BouncePawn _asfort_pawn _output_get))
        (:CastFailed
          (CallFunction|BounceOther Actor))))
    (:then_1
      (CallFunction|OnBounced Actor))))

(fn ExecuteBounceCue
  (Networking|CanExecuteCosmeticEvents
    (:True
      (Development|PrintString "Bouncy Prop: bounce cue")
      (GameplayCue|ExecuteGameplayCue%28Burst%29Local self (Variables|BounceProp|FX|GetCue_Bounced) (Ability|GameplayCue|MakeGameplayCueParameters :NormalizedMagnitude 0.0 :RawMagnitude 0.0 :Location (Transformation|GetWorldLocation (Variables|Default|GetBounceFXPoint)) :Normal (Transformation|GetActorUpVector) :GameplayEffectLevel 1 :AbilityLevel 1 :bReplicateLocationWhenUsingMinimalRepProxy false)))))

(fn GetBaseBounceDirection
  (return (Math|Vector|SelectVector (Math|Vector|RotateVector (Variables|BounceProp|GetBaseBounceDirection) (Transformation|GetActorRotation)) (Variables|BounceProp|GetBaseBounceDirection) (Variables|BounceProp|GetbBaseDirectionIsRelative))))

(fn GetBounceMagnitude
  (return (+ (ScalableFloat|GetValueatLevel (Variables|BounceProp|GetBounceMagnitude)) (Variables|BounceProp|GetInstanceBounceMagnitudeAdjustment))))

(fn GetBounceVelocity
  (params (ActorToBounce :type Actor%20Object%20Reference))
  (bind (_returnvalue_x _returnvalue_y) (Math|Vector|BreakVector (Transformation|GetVelocity ActorToBounce)))
  (return (+ (Utilities|Operators|Multiply (Math|Vector|MakeVector _returnvalue_x _returnvalue_y) (ScalableFloat|GetValueatLevel (Variables|BounceProp|GetBouncerVelocityMaintained))) (* (CallFunction|GetBaseBounceDirection) (CallFunction|GetBounceMagnitude)))))

(fn HandleJumpTriggerOverlap
  (params (Actor :type Actor%20Object%20Reference))
  (if (CallFunction|CanBounceActor Actor)
    (CallFunction|DoBounce Actor)))

(fn OnBounced
  (params (BouncedActor :type Actor%20Object%20Reference))
  (CallFunction|ApplyBouncedGameplayEffects BouncedActor)
  (Networking|SwitchHasAuthority
    (:Authority
      (CallFunction|PlayBounceFX))))

(fn ConstructionScript)
